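# Idle Niagara effect for the HayBrush pickaxe: a mesh emitter (SoupDrop) drips sparse droplets from baked local points, stopping and hiding them while swinging.
# Each drop hangs on a lengthening pendulum from an array-picked anchor, switching between constrained and free position around 80% of its life.
# Gravity and late-life acceleration against horizontal owner velocity move the drops.
# Meshes stretch with age and speed, and their color follows time of day.
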NiagaraSystem /BRCosmetics/Weapons/FORT_Melee/Pickaxe_HayBrush/FX/NS_Pickaxe_HayBrush_Idle
dependencies {
 usedDataInterfaces [
  &NiagaraDataInterfaceArrayFloat3
  &NiagaraDataInterfaceColorCurve
  &NiagaraDataInterfaceCurve
  &NiagaraDataInterfaceSkeletalMesh
 ]
 usedDynamicInputs []
 usedModules [
  &self:SystemSpawnScript.SystemScriptSource.SystemScriptGraph.NiagaraNodeAssignment_0.SetVariables_#1
  &SystemState
 ]
}
emitters {SoupDrop {
  emitterUpdateScript [
   &EmitterState(Life Cycle Mode=::NewEnumerator1, UseLoopDelay=false)
   &SpawnRate(
    Spawn Probability=0.5
    SpawnRate=&MakeCustomFloatFromBool(Bool=$User.bisSwinging, True Float=0, False Float=1.5)
    Use Spawn Probability=true
   )
  ]
  particleSpawnScript [
   /SetVariables(Particles.DropType:NiagaraFloat=&RandomRangeFloat(Minimum=0, Maximum=1)) {isSetParameters=true}
   &InitializeParticle(
    Lifetime=&Lerp_Float(A=3, B=5, Alpha=$Particles.DropType)
    Mesh Scale Mode=::NewEnumerator3
    Mesh Uniform Scale=&Multiply_Float(
     A=&Lerp_Float(A=0.05, B=0.1, Alpha=$Particles.DropType)
     B=&FN_FrontEndFloatModifier(Gameplay=1, Front End=0.8)
    )
    Position Offset=&SelectVectorFromArray(
     Array Sampling Mode=::NewEnumerator0
     Direct Array Index=?
     Vector Selection Array=DataInterface() {propertyValues {FloatData [
        <-10.0807 -9.0122 63.7347>
        <-10.6321 -7.2438 53.8567>
        <-9.6089 -7.3476 57.7274>
        <-5.3317 6.6371 80.6219>
        <-11.1063 4.6445 72.2099>
        <-7.0891 -8.1927 79.0094>
        <-9.563 -9.3115 74.4029>
        <-11.2572 -6.1242 56.5327>
        <-9.4585 10.3595 70.7847>
        <-8.2142 -10.7577 73.1195>
        <-7.3347 7.9123 78.5285>
        <-9.6991 8.2117 73.4858>
        <-10.2163 -7.9741 61.9998>
        <-4.1746 -1.7841 83.2496>
        <-8.3402 -8.521 76.4366>
        <-2.6425 0.1587 82.0695>
        <-11.4221 -8.8839 64.7864>
        <-10.3001 -9.4843 71.7172>
        <-11.0478 3.1805 73.0692>
        <-7.1981 9.21 73.8807>
        <-11.9248 -9.0016 62.4232>
        <-3.4066 4.8966 80.8647>
        <-4.5619 -7.8326 78.9876>
        <-11.6789 -9.813 58.0944>
        <-8.8412 -8.9505 71.0324>
        <-11.6327 -7.3506 60.1164>
        <-2.9818 2.0781 83.3262>
        <-3.9062 -7.0422 81.2331>
        <-10.1537 3.0008 74.7723>
        <-5.8973 1.9681 78.7647>
        <-10.4739 -9.0354 66.4693>
        <-7.2356 4.185 80.0956>
        <-8.9319 -9.8093 68.5375>
        <-3.437 3.9636 82.4888>
        <-7.5374 5.7062 75.7161>
        <-5.4666 -8.2544 77.5736>
        <-5.2247 7.9702 77.7348>
        <-7.2202 9.4701 75.9927>
        <-11.0191 -9.9368 67.6874>
        <-6.0204 5.0006 77.3719>
        <-8.2016 2.5435 76.8526>
        <-6.6314 -7.3132 76.6745>
        <-6.3729 -0.9231 80.0169>
        <-10.675 6.3284 75.8864>
        <-4.1558 -5.8175 82.2306>
        <-6.5512 -5.6798 79.9464>
        <-6.5359 -2.0198 78.3692>
        <-8.7331 -9.8789 69.8335>
        <-7.4649 -10.2352 74.8597>
        <-5.8033 1.6638 81.3498>
       ]}}
     Random Seed=0
    )
    Position Offset Coordinate Space=::Local
    UsePositionOffset=true
    Write Material Random=true
   )
  ]
  particleUpdateScript [
   &ParticleState()
   /SetVariables(Particles.ConnectionTime:NiagaraBool=&SetBoolByFloatComparison(A=$Particles.NormalizedAge, B=0.8)) {isSetParameters=true}
   /SetVariables(Particles.Color=&ColorFromCurve(ColorCurve=DataInterface() {propertyValues {
       AlphaCurve {keys [<at 0 1 RCIM_Cubic>]}
       BlueCurve {keys [<at 0 0.54 RCIM_Cubic>]}
       GreenCurve {keys [<at 0 0.565 RCIM_Cubic>]}
       RedCurve {keys [<at 0 0.6 RCIM_Cubic>]}
      }})) {isSetParameters=true}
   /SetVariables(Particles.Color=&MultiplyLinearColor(
     Linear Color=$Particles.Color
     Scale Factor=&MakeVector4_FromLinearColor(LinearColor=&FNTimeOfDayColor(Night Color=<rgba 0.4 0.4667 0.8 1>, Morning Color=<rgba 0.9 0.716 0.6058 1>, Day Color=<rgba 1 1 1 1>, Evening Color=<rgba 1 0.6667 0.5 1>))
    )) {isSetParameters=true}
   /SetVariables(Particles.AnchorPos:NiagaraPosition=&SelectVectorFromArray(
     Array Sampling Mode=::NewEnumerator0
     Direct Array Index=?
     Vector Selection Array=DataInterface() {propertyValues {FloatData [
        <-11.2562 2.5743 55.0624>
        <-12.1601 4.2428 54.0866>
        <-13.3447 3.1032 56.2772>
        <-13.4409 1.6519 59.7777>
        <-12.2323 2.1995 59.2585>
        <-10.5767 3.2443 55.125>
        <-13.7533 2.4353 56.8906>
        <-5.1829 9.9342 75.2211>
        <-5.2659 9.7612 75.6994>
        <-3.1865 11.8752 73.657>
        <-3.0907 11.579 73.7602>
        <-3.1298 11.8727 73.6147>
        <0.0926 11.3225 75.0254>
        <0.8115 11.61 74.5774>
        <0.7398 12.3258 76.8657>
        <0.9689 12.0086 77.5134>
        <0.8851 10.2705 77.2955>
        <-1.9848 7.9729 78.4282>
        <-2.3048 7.4984 78.7227>
        <-0.5259 6.0307 80.5215>
        <-3.1768 11.1208 78.0847>
        <-2.4427 8.9003 81.2702>
        <-3.8587 6.7858 81.7536>
        <-3.6367 4.1615 83.4694>
        <-5.5968 4.3771 82.1335>
        <-4.2993 3.9276 81.2215>
        <-5.5019 5.1568 79.7394>
        <-6.613 7.2028 77.4191>
        <-7.2985 7.6424 77.3419>
        <-3.9602 6.1186 79.9038>
        <-2.7765 5.7048 79.761>
        <-0.8596 5.9135 83.3072>
        <-12.0024 2.5362 51.6133>
        <-11.8675 3.0624 52.7027>
        <-0.304 14.1424 74.7474>
        <-0.4468 13.828 75.4865>
        <-2.4065 13.0312 75.2175>
        <-0.47 14.0047 72.8769>
        <-0.408 14.1686 72.7063>
        <-14.6432 -0.0636 73.9759>
        <-13.5967 -0.8479 74.9161>
        <-7.9725 -1.1396 81.6735>
        <-8.5264 -1.1052 81.6821>
        <-12.9048 -1.1138 75.61>
        <-12.6111 -1.0533 73.4786>
        <-14.1862 -1.2372 72.0143>
        <-10.5872 0.5334 76.2387>
        <-8.4255 -1.0878 79.2181>
       ]}}
     Random Seed=0
    )) {isSetParameters=true}
   &GravityForce(Gravity=<0 0 -600>)
   &PendulumSetup(
    Module.Enable Spring Driver=false
    Pendulum Length:NiagaraFloat=&FloatFromCurve(
     FloatCurve=DataInterface() {propertyValues {Curve {keys [<at 0 0 RCIM_Cubic>, {arriveTangent=0.0714, interpMode=RCIM_Cubic, leaveTangent=0.0714, tangentMode=RCTM_User, time=1, value=1}]}}}
     Scale Curve=0.5
    )
    Pendulum Pivot:NiagaraPosition=$Particles.AnchorPos
    "Spring Driven Constraint (Tightness)"=2
   )
   &AccelerationForce(Acceleration=&Multiply_VectorByFloat(
     Vector=&Multiply_Vector(A=$Engine.Owner.Velocity, B=<-1 -1 0>)
     Float=&FloatFromCurve(FloatCurve=DataInterface() {propertyValues {Curve {keys [{interpMode=RCIM_Cubic, tangentMode=RCTM_User, time=0.6, value=0}, {interpMode=RCIM_Cubic, tangentMode=RCTM_User, time=0.9, value=1}]}}})
    ))
   &SolveForcesAndVelocity(Clamp Velocity=false, Limit Acceleration=false)
   /SetVariables(Particles.PositionTemp:NiagaraPosition=$Particles.Position) {isSetParameters=true}
   &PendulumConstraint()
   /SetVariables(Particles.Position=&MakeCustomVectorFromBool(
     Bool=$Particles.ConnectionTime
     True Vector=Linked() {linkedVariable {name=Particles.PositionTemp, type=NiagaraPosition}}
     False Vector=Linked() {linkedVariable {name=Particles.Position, type=NiagaraPosition}}
    )) {isSetParameters=true}
   &CalculateAccurateVelocity()
   /SetVariables(Particles.MeshOrientation=&OrientMeshToVector(
     Look At Direction=&Lerp_Vector(
      A=&TransformVector(Vector=<0 0 1>, Source Space=::World, Destination Space=::Local)
      B=&TransformVector(
       Vector=&Divide_Vector(A:Vector3f=$Engine.Owner.Velocity, B:Vector3f=&VectorFromFloat(Value=2))
       Source Space=::World
       Destination Space=::Local
      )
      Alpha=&Multiply_Float(A=$Engine.DeltaTime, B=0.1)
     )
     Reference Axis Vector=<0 0 1>
    )) {isSetParameters=true}
   &ScaleMeshSize(Scale Factor=&VectorFromFloat(Value=&FloatFromCurve(
      FloatCurve=DataInterface() {propertyValues {Curve {keys [
          {arriveTangent=3.0781, interpMode=RCIM_Cubic, leaveTangent=3.0781, tangentMode=RCTM_User, time=0, value=0}
          {arriveTangent=-0.0171, interpMode=RCIM_Cubic, leaveTangent=-0.0171, tangentMode=RCTM_User, time=0.95, value=1}
          {arriveTangent=-22.3504, interpMode=RCIM_Cubic, leaveTangent=-22.3504, tangentMode=RCTM_User, time=1, value=0}
         ]}}}
      Scale Curve=&MakeCustomFloatFromBool(Bool=$User.bisSwinging, True Float=0, False Float=1)
     )))
   &ScaleMeshSize(Scale Factor=&MakeVector(
     X=1
     Y=1
     Z=&FloatFromCurve(FloatCurve=DataInterface() {propertyValues {Curve {keys [
          {arriveTangent=2.6357, interpMode=RCIM_Cubic, leaveTangent=2.6357, tangentMode=RCTM_User, time=0.2, value=1}
          {arriveTangent=0.0393, interpMode=RCIM_Cubic, leaveTangent=0.0393, tangentMode=RCTM_User, time=1, value=2}
         ]}}})
    )) {module=ScaleMeshSize001}
   &ScaleMeshSizeBySpeed(Max Scale Factor=<1 1 1.5>, Min Scale Factor:Vector3f=<0.6 0.6 1>, Speed Threshold=300)
  ]
  rendererClasses [&NiagaraMeshRendererProperties]
  renderers [{rendererClass=&NiagaraMeshRendererProperties, rendererIndex=0}]
 }}
summary {
 systemName=NS_Pickaxe_HayBrush_Idle
 emitters [{emitterName=SoupDrop, rendererClasses [&NiagaraMeshRendererProperties]}]
 userVariables [
  User.bIsFrontEnd {type=NiagaraBool}
  User.bIsFrontEndPreview {type=NiagaraBool}
  User.bisSwinging {type=NiagaraBool}
  User.IsSkydiving {type=NiagaraBool}
 ]
}
