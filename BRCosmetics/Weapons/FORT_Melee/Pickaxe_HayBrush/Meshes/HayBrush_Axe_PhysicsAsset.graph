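# Physics asset for the HayBrush pickaxe mesh, a two-handed melee cosmetic weapon.
# It defines a single body on the Melee_TwoHanded_Handle bone, shaped as one capsule rotated and offset along the weapon.
# It has no constraints, so the asset contains no articulated or secondary-motion joints.
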PhysicsAsset /BRCosmetics/Weapons/FORT_Melee/Pickaxe_HayBrush/Meshes/HayBrush_Axe_PhysicsAsset
bodies {Melee_TwoHanded_Handle [{shapeType=Capsule, length=55.7862, radius=11.691, center=<1.1818 -4.2028 28.4348>, rotation=<rot -6.3902 -90 -0.0731>}]}
constraints []
